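# Editor utility widget serving as one row in the container replacement tool, pairing a container label with a picker for a replacement asset.
# On PreConstruct, it binds its embedded DetailsView to itself, showing only the AssetToReplaceWith property as an editable object picker.
# SetEntryId writes the given string to the bold container label.
# A ContainerObjectToMap variable is declared, but no graph here uses it.
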EditorUtilityWidgetBlueprint /FigmentTools/ContainersReplacementTool/EditorUtilityWidget/WBP_ContainerReplacementEntry
graphs [
 EventGraph {nodes [
   Event PreConstruct()
   > SetObject(self=Get DetailsView_ObjectPicker() -> (DetailsView_ObjectPicker:Details View Object Reference), NewObject=Self-Reference())
  ]}
 SetEntryId {nodes [
   n1 = SetEntryId!() -> (then, Id:String) {is=title K2Node_FunctionEntry}
   > "SetText (Text)"(self=Get TextBlock_Container() -> (TextBlock_Container:Text Object Reference), InText="To Text (String)"(InString=@n1))
  ]}
]
variables [
 AssetToReplaceWith {type=Object, default=null}
 ContainerObjectToMap {type=Object, default=null}
]
widgets {
 info {rootWidgetClass=&VerticalBox, widgetCount=8}
 widgets [
  VerticalBox_1 {widgetClassPath=&VerticalBox, widget=VerticalBox_1}
  HorizontalBox_40 {widgetClassPath=&HorizontalBox, parent=VerticalBox_1, widget=HorizontalBox_40, slot {class=VerticalBoxSlot}}
  Border_0 {
   widgetClassPath=&Border
   parent=HorizontalBox_40
   widget=Border_0
   properties {brushColor=<rgba 0.0285 0.0312 0.0251 1>, padding=<ltrb 6 6 6 6>}
   slot {class=HorizontalBoxSlot, properties {size {sizeRule=Fill, value=1}}}
  }
  HorizontalBox_0 {
   widgetClassPath=&HorizontalBox
   parent=Border_0
   widget=HorizontalBox_0
   slot {class=BorderSlot, properties {padding=<ltrb 6 6 6 6>}}
  }
  TextBlock_Container {
   widgetClassPath=&TextBlock
   bIsVariable=true
   parent=HorizontalBox_0
   widget=TextBlock_Container
   properties {
    text=Container
    font {
     bForceMonospaced=false
     bMaterialIsStencil=false
     fontMaterial=null
     fontObject=&Roboto
     letterSpacing=0
     monospacedWidth=1
     size=10.5
     skewAmount=0
     typefaceFontName=Bold
     outlineSettings {
      bApplyOutlineToDropShadows=false
      bMiteredCorners=false
      bSeparateFillAlpha=false
      outlineMaterial=null
      outlineSize=0
      outlineColor=<rgba 0 0 0 1>
     }
    }
   }
   slot {class=HorizontalBoxSlot, properties {horizontalAlignment=HAlign_Center, verticalAlignment=VAlign_Center}}
  }
  Border_1 {
   widgetClassPath=&Border
   parent=HorizontalBox_0
   widget=Border_1
   properties {brushColor=<rgba 0.3073 0.3073 0.3073 1>, padding=<ltrb 2 2 2 2>}
   slot {class=HorizontalBoxSlot, properties {padding=<ltrb 16 0 0 0>, size {sizeRule=Fill, value=1}}}
  }
  DetailsView_ObjectPicker {
   widgetClassPath=&DetailsView
   parent=Border_1
   widget=DetailsView_ObjectPicker
   properties {columnWidth=0.6, propertiesToShow [AssetToReplaceWith]}
   slot {class=BorderSlot, properties {padding=<ltrb 2 2 2 2>}}
  }
  Spacer_486 {widgetClassPath=&Spacer, parent=VerticalBox_1, widget=Spacer_486, properties {size=<1 16>}, slot {class=VerticalBoxSlot}}
 ]
}
