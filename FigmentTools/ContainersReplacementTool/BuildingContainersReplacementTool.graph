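# Editor utility widget that finds building containers in chosen levels and replaces them with selected assets.
# A ContainerReplacementToolModel built on PreConstruct does the work and backs a details view of levels to scan, filled from a directory or selected assets.
# Search passes a data-asset class filter and include/ignore mode to the model, then lists results as entry widgets.
# Replace copies each entry's choice to the model, showing an error if none is set.
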EditorUtilityWidgetBlueprint /FigmentTools/ContainersReplacementTool/BuildingContainersReplacementTool
graphs [
 EventGraph {nodes [
   "On Clicked (EditorUtilityButton_Execute)"() {is=title K2Node_ComponentBoundEvent}
   > UpdateData()
   > Execute(self:Container Replacement Tool Model Object Reference=Get ContainersReplacementModel() -> (ContainersReplacementModel:Container Replacement Tool Model Object Reference))
   > ResetInterface()
   > UpdateContainersInterface()
   "On Clicked (EditorUtilityButton_Replace)"() {is=title K2Node_ComponentBoundEvent}
   > SetReplacementsInformation()
   n1 = Get ContainersReplacementModel()
   > n2 = Branch(Condition=CanPerformReplacement!(self:Container Replacement Tool Model Object Reference=@n1) -> (ReturnValue:bool))
   > PerformAssetsReplacement(self:Container Replacement Tool Model Object Reference=@n1)
   > ResetInterface()
   Show Message Dialog(
    execute=@n2.else
    Title=MakeLiteralText(Value="LOCTEXT(\"Building Container Replacement\")")
    Message=MakeLiteralText(Value="LOCTEXT(\"At least one object needs to have its replacement selected.\")")
    MessageType=Ok
    DefaultValue=Ok
    MessageCategory=Error
   )
   "On Clicked (Button_AddLevelsFromDirectory)"() {is=title K2Node_ComponentBoundEvent}
   > AddLevelsFromDirectory(self:Container Replacement Tool Model Object Reference=Get ContainersReplacementModel())
   "On Clicked (Button_AddSelectedLevels)"() {is=title K2Node_ComponentBoundEvent}
   > AddLevelsFromSelectedAssets(self:Container Replacement Tool Model Object Reference=Get ContainersReplacementModel())
   Event PreConstruct()
   > n3 = Construct Object from Class(Class=/Script/FigmentTools.ContainerReplacementToolModel) -> (ReturnValue:Container Replacement Tool Model Object Reference)
   > n4 = Set ContainersReplacementModel(ContainersReplacementModel:Container Replacement Tool Model Object Reference=@n3) -> (Output_Get:Container Replacement Tool Model Object Reference)
   > SetObject(self=Get DetailsView_Levels() -> (DetailsView_Levels:Details View Object Reference), NewObject=@n4)
   > SetObject(
    self=Get SinglePropertyView_ClassesFilter() -> (SinglePropertyView_ClassesFilter:Single Property View Object Reference)
    NewObject=Self-Reference()
   )
  ]}
 ResetInterface {nodes [
   ResetInterface() {is=title K2Node_FunctionEntry}
   n1 = Get FoundContainers() -> (FoundContainers:Array of WBP Container Replacement Entry Object References)
   > n2 = For Each Loop(Array:Array of WBP Container Replacement Entry Object References=@n1) -> (Array Element:WBP Container Replacement Entry Object Reference)
   Clear(execute=@n2.Completed, TargetArray:Array of WBP Container Replacement Entry Object References=@n1)
   > SetVisibility(self=Get VerticalBox_ReplacementOperation() -> (VerticalBox_ReplacementOperation:Vertical Box Object Reference), InVisibility=Hidden)
   RemoveFromParent(execute=@n2.LoopBody, self=@n2)
  ]}
 SetReplacementsInformation {nodes [
   SetReplacementsInformation() {is=title K2Node_FunctionEntry}
   > n1 = For Each Loop(Array=Get FoundContainers())
   > SetContainerReplacement!(
    execute
    self:Container Replacement Tool Model Object Reference=Get ContainersReplacementModel()
    InObjectToBeReplaced:Object Reference=Get ContainerObjectToMap!(self:WBP Container Replacement Entry Object Reference=@n1) -> (ContainerObjectToMap:Object Reference)
    InReplacement:Object Reference=Get AssetToReplaceWith!(self:WBP Container Replacement Entry Object Reference=@n1) -> (AssetToReplaceWith:Object Reference)
   ) -> (then)
  ]}
 UpdateContainersInterface {nodes [
   UpdateContainersInterface() {is=title K2Node_FunctionEntry}
   > n1 = For Each Loop(Array:Array of Object References=Get ContainersObjectsToBeReplaced!(self:Container Replacement Tool Model Object Reference=Get ContainersReplacementModel()) -> (ContainersObjectsToBeReplaced:Array of Object References)) -> (Array Element:Object Reference)
   n2 = Get FoundContainers()
   Branch(execute=@n1.Completed, Condition="integer > integer"(A=Length(TargetArray:Array of WBP Container Replacement Entry Object References=@n2)))
   > SetVisibility(self=Get VerticalBox_ReplacementOperation(), InVisibility=Visible)
   n3 = Construct Object from Class(
    execute=@n1.LoopBody
    Class=/FigmentTools/ContainersReplacementTool/EditorUtilityWidget/WBP_ContainerReplacementEntry.WBP_ContainerReplacementEntry_C
   ) -> (ReturnValue:WBP Container Replacement Entry Object Reference)
   > Set ContainerObjectToMap!(execute, ContainerObjectToMap:Object Reference=@n1, self:WBP Container Replacement Entry Object Reference=@n3) -> (then, Output_Get:Object Reference)
   > AddChild(self=Get VerticalBox_ContainersToMap() -> (VerticalBox_ContainersToMap:Editor Utility Scroll Box Object Reference), Content=@n3)
   > SetEntryId!(execute, self:WBP Container Replacement Entry Object Reference=@n3, Id:String=GetDisplayName(Object=@n1)) -> (then)
   > Add(
    TargetArray:Array of WBP Container Replacement Entry Object References=Get FoundContainers()
    NewItem:"WBP Container Replacement Entry Object Reference (by ref)"=@n3
   ) {is=title K2Node_CallArrayFunction}
  ]}
 UpdateData {nodes [
   UpdateData() {is=title K2Node_FunctionEntry}
   n1 = Get ContainersReplacementModel()
   > Set ClassesFilter!(
    execute
    ClassesFilter:Array of Actor Class References=Get ClassesFilter!(self:Container Replacement Data Asset Object Reference=Get ClassesFilterDataAsset() -> (ClassesFilterDataAsset:Container Replacement Data Asset Object Reference)) -> (ClassesFilter:Array of Actor Class References)
    self:Container Replacement Tool Model Object Reference=@n1
   ) -> (then, Output_Get:Array of Actor Class References)
   n2 = Get EditorUtilityComboBoxString_Operation() -> (EditorUtilityComboBoxString_Operation:Editor Utility Combo Box String Object Reference)
   > Set ClassesFilterOperation!(
    execute
    ClassesFilterOperation:EClassesFilterOperation Enum=Select(
     Option 0:EClassesFilterOperation Enum=IncludeOnly
     Option 1:EClassesFilterOperation Enum=IgnoreAll
     Index:int=GetSelectedIndex!(self:"Combo Box (String) Object Reference"=@n2) -> (ReturnValue:int=0)
    ) -> (ReturnValue:EClassesFilterOperation Enum)
    self:Container Replacement Tool Model Object Reference=@n1
   ) -> (then, Output_Get:EClassesFilterOperation Enum=IncludeOnly)
  ]}
]
variables [
 ClassesFilter {type="array<Class>", default [null]}
 ClassesFilterDataAsset {type=ContainerReplacementDataAsset, default=/FigmentTools/ContainersReplacementTool/DataAssets/DA_BuildingContainerReplacement}
 ContainersReplacementModel {type=ContainerReplacementToolModel, default=null}
 FoundContainers {type="array<WBP_ContainerReplacementEntry_C>", default []}
 Recursive {type=boolean}
]
widgets {
 info {rootWidgetClass=&VerticalBox, widgetCount=22}
 widgets [
  VerticalBox_0 {widgetClassPath=&VerticalBox, widget=VerticalBox_0}
  TestTool {
   widgetClassPath=&TextBlock
   parent=VerticalBox_0
   widget=TestTool
   properties {
    justification=Center
    text=Containers Replacement
    font {
     bForceMonospaced=false
     bMaterialIsStencil=false
     fontMaterial=null
     fontObject=&Roboto
     letterSpacing=0
     monospacedWidth=1
     size=16.5
     skewAmount=0
     typefaceFontName=Bold
     outlineSettings {
      bApplyOutlineToDropShadows=false
      bMiteredCorners=false
      bSeparateFillAlpha=false
      outlineMaterial=null
      outlineSize=0
      outlineColor=<rgba 0 0 0 1>
     }
    }
   }
   slot {class=VerticalBoxSlot, properties {verticalAlignment=VAlign_Top}}
  }
  HorizontalBox_0 {widgetClassPath=&HorizontalBox, parent=VerticalBox_0, widget=HorizontalBox_0, slot {class=VerticalBoxSlot}}
  SinglePropertyView_ClassesFilter {
   widgetClassPath=&SinglePropertyView
   parent=HorizontalBox_0
   widget=SinglePropertyView_ClassesFilter
   properties {propertyName=ClassesFilterDataAsset}
   slot {class=HorizontalBoxSlot, properties {size {sizeRule=Fill, value=1}}}
  }
  EditorUtilityComboBoxString_Operation {
   widgetClassPath=/Script/Blutility.EditorUtilityComboBoxString
   bIsVariable=true
   parent=HorizontalBox_0
   widget=EditorUtilityComboBoxString_Operation
   properties {
    selectedOption=Ignore All
    defaultOptions [Include Only, Ignore All]
    foregroundColor {colorUseRule=UseColor_Specified, specifiedColor=<rgba 0.5271 0.5271 0.5271 1>}
   }
   slot {class=HorizontalBoxSlot}
  }
  HorizontalBox_106 {widgetClassPath=&HorizontalBox, parent=VerticalBox_0, widget=HorizontalBox_106, slot {class=VerticalBoxSlot}}
  TextBlock_147 {
   widgetClassPath=&TextBlock
   parent=HorizontalBox_106
   widget=TextBlock_147
   properties {
    text=Add Levels
    font {
     bForceMonospaced=false
     bMaterialIsStencil=false
     fontMaterial=null
     fontObject=&Roboto
     letterSpacing=0
     monospacedWidth=1
     size=12
     skewAmount=0
     typefaceFontName=Regular
     outlineSettings {
      bApplyOutlineToDropShadows=false
      bMiteredCorners=false
      bSeparateFillAlpha=false
      outlineMaterial=null
      outlineSize=0
      outlineColor=<rgba 0 0 0 1>
     }
    }
   }
   slot {class=HorizontalBoxSlot}
  }
  Button_AddLevelsFromDirectory {
   widgetClassPath=&EditorUtilityButton
   parent=HorizontalBox_106
   widget=Button_AddLevelsFromDirectory
   properties {toolTipText=Add levels from a directory}
   slot {class=HorizontalBoxSlot}
  }
  Image_71 {
   widgetClassPath=&Image
   bIsVariable=true
   parent=Button_AddLevelsFromDirectory
   widget=Image_71
   properties {brush {
     bIsDynamicallyLoaded=false
     drawAs=Image
     imageType=FullColor
     mirroring=NoMirror
     resourceName=None
     resourceObject=/Game/Athena/UI/InGame/Creative/PropertyEditor/Icon_Picker/Textures_V2/64x64/T_UI_IconLibrary_Folder_64
     tiling=NoTile
     imageSize=<16 16>
     margin=<ltrb 0 0 0 0>
     outlineSettings {
      bUseBrushTransparency=false
      roundingType=HalfHeightRadius
      width=0
      color {colorUseRule=UseColor_Specified, specifiedColor=<rgba 0 0 0 0>}
      cornerRadii=<0 0 0 0>
     }
     tintColor {colorUseRule=UseColor_Specified, specifiedColor=<rgba 1 1 1 1>}
     uVRegion {bIsValid=false, max=<0 0>, min=<0 0>}
    }}
   slot {class=ButtonSlot}
  }
  Button_AddSelectedLevels {
   widgetClassPath=&EditorUtilityButton
   parent=HorizontalBox_106
   widget=Button_AddSelectedLevels
   properties {toolTipText=Add selected levels from the content drawer}
   slot {class=HorizontalBoxSlot}
  }
  Image {
   widgetClassPath=&Image
   bIsVariable=true
   parent=Button_AddSelectedLevels
   widget=Image
   properties {brush {
     bIsDynamicallyLoaded=false
     drawAs=Image
     imageType=FullColor
     mirroring=NoMirror
     resourceName=None
     resourceObject=/Game/Athena/UI/InGame/Creative/PropertyEditor/Icon_Picker/Textures_V2/64x64/T_UI_IconLibrary_LeftArrow_64
     tiling=NoTile
     imageSize=<16 16>
     margin=<ltrb 0 0 0 0>
     outlineSettings {
      bUseBrushTransparency=false
      roundingType=HalfHeightRadius
      width=0
      color {colorUseRule=UseColor_Specified, specifiedColor=<rgba 0 0 0 0>}
      cornerRadii=<0 0 0 0>
     }
     tintColor {colorUseRule=UseColor_Specified, specifiedColor=<rgba 1 1 1 1>}
     uVRegion {bIsValid=false, max=<0 0>, min=<0 0>}
    }}
   slot {class=ButtonSlot}
  }
  SizeBox_271 {
   widgetClassPath=&SizeBox
   parent=VerticalBox_0
   widget=SizeBox_271
   properties {bOverride_MaxDesiredHeight=true, bOverride_MinDesiredHeight=true, heightOverride=500, maxDesiredHeight=500, minDesiredHeight=150}
   slot {class=VerticalBoxSlot, properties {size {sizeRule=Fill, value=0.5}}}
  }
  DetailsView_Levels {
   widgetClassPath=&DetailsView
   parent=SizeBox_271
   widget=DetailsView_Levels
   properties {bAllowFiltering=true, bAutoLoadAsset=false, propertiesToShow [LevelsToScan]}
   slot {class=SizeBoxSlot}
  }
  EditorUtilityButton_Execute {
   widgetClassPath=&EditorUtilityButton
   parent=VerticalBox_0
   widget=EditorUtilityButton_Execute
   slot {class=VerticalBoxSlot, properties {verticalAlignment=VAlign_Bottom, padding=<ltrb 5 5 5 5>}}
  }
  TextBlock_31 {
   widgetClassPath=&TextBlock
   parent=EditorUtilityButton_Execute
   widget=TextBlock_31
   properties {
    text=Search
    font {
     bForceMonospaced=false
     bMaterialIsStencil=false
     fontMaterial=null
     fontObject=&Roboto
     letterSpacing=0
     monospacedWidth=1
     size=12
     skewAmount=0
     typefaceFontName=Bold
     outlineSettings {
      bApplyOutlineToDropShadows=false
      bMiteredCorners=false
      bSeparateFillAlpha=false
      outlineMaterial=null
      outlineSize=0
      outlineColor=<rgba 0 0 0 1>
     }
    }
   }
   slot {class=ButtonSlot}
  }
  VerticalBox_ReplacementOperation {
   widgetClassPath=&VerticalBox
   bIsVariable=true
   parent=VerticalBox_0
   widget=VerticalBox_ReplacementOperation
   properties {visibility=Collapsed}
   slot {class=VerticalBoxSlot, properties {size {sizeRule=Fill, value=1}}}
  }
  VerticalBox_1 {
   widgetClassPath=&VerticalBox
   parent=VerticalBox_ReplacementOperation
   widget=VerticalBox_1
   slot {class=VerticalBoxSlot, properties {padding=<ltrb 0 10 0 0>, size {sizeRule=Fill, value=1}}}
  }
  Text_FoundContainers {
   widgetClassPath=&TextBlock
   parent=VerticalBox_1
   widget=Text_FoundContainers
   properties {
    text=Building Containers found
    font {
     bForceMonospaced=false
     bMaterialIsStencil=false
     fontMaterial=null
     fontObject=&Roboto
     letterSpacing=0
     monospacedWidth=1
     size=13.5
     skewAmount=0
     typefaceFontName=Bold
     outlineSettings {
      bApplyOutlineToDropShadows=false
      bMiteredCorners=false
      bSeparateFillAlpha=false
      outlineMaterial=null
      outlineSize=0
      outlineColor=<rgba 0 0 0 1>
     }
    }
   }
   slot {class=VerticalBoxSlot, properties {horizontalAlignment=HAlign_Center, verticalAlignment=VAlign_Top}}
  }
  Spacer_152 {widgetClassPath=&Spacer, parent=VerticalBox_1, widget=Spacer_152, properties {size=<649.3942 12>}, slot {class=VerticalBoxSlot}}
  VerticalBox_ContainersToMap {
   widgetClassPath=&EditorUtilityScrollBox
   bIsVariable=true
   parent=VerticalBox_1
   widget=VerticalBox_ContainersToMap
   slot {class=VerticalBoxSlot, properties {size {sizeRule=Fill, value=1}}}
  }
  EditorUtilityButton_Replace {
   widgetClassPath=&EditorUtilityButton
   parent=VerticalBox_ReplacementOperation
   widget=EditorUtilityButton_Replace
   slot {class=VerticalBoxSlot, properties {verticalAlignment=VAlign_Bottom, padding=<ltrb 0 12 0 0>}}
  }
  Text_Replace {
   widgetClassPath=&TextBlock
   parent=EditorUtilityButton_Replace
   widget=Text_Replace
   properties {
    text=Replace Containers
    font {
     bForceMonospaced=false
     bMaterialIsStencil=false
     fontMaterial=null
     fontObject=&Roboto
     letterSpacing=0
     monospacedWidth=1
     size=10.5
     skewAmount=0
     typefaceFontName=Bold
     outlineSettings {
      bApplyOutlineToDropShadows=false
      bMiteredCorners=false
      bSeparateFillAlpha=false
      outlineMaterial=null
      outlineSize=0
      outlineColor=<rgba 0 0 0 1>
     }
    }
   }
   slot {class=ButtonSlot}
  }
 ]
}
